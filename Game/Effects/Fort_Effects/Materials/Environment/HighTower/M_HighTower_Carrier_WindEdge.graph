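# Translucent, unlit, two-sided wind-edge effect material for the HighTower Carrier, tinted by time of day.
# Opacity layers panning, distorted noise textures, masked by axis-aligned fresnel, a V gradient and near/far camera fades.
# World position offset adds noise-driven wobble along the V gradient and throws vertices far down beyond the buffered far-fade distance.
# MPC_Butterfly's HideWind parameter fades opacity out and, at full value, also drops the mesh out of view.
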Material /Game/Effects/Fort_Effects/Materials/Environment/HighTower/M_HighTower_Carrier_WindEdge
graph {
 properties {
  bUsedWithNanite=true
  blendMode=BLEND_Translucent
  pixelDepthOffsetMode=PDOM_Legacy
  shadingModel=MSM_Unlit
  twoSided=true
  usedShadingModels=Unlit
 }
 nodes [
  DynamicParameter() {defaultValue=<rgba 10 0 1 1>, paramNames [WPO, Erode, PanSeed_V, CurvePanAnim_V]}
  n1 = CollectionParameter() {collection=/Game/Athena/Environments/Nexus/Materials/MPC_Butterfly, parameterName=HideWind}
  n2 = Constant3Vector() {constant=<rgba 0 0 -500000 0>}
  n3 = MaterialFunctionCall(Vector 2=ComponentMask(None=WorldPosition() {worldPositionShaderOffset=WPT_CameraRelative}.XYZ) {g=true, r=true}) {materialFunction=/Engine/Functions/Engine_MaterialFunctions02/Utility/VectorLength}
  n4 = Multiply(A=ScalarParameter() {defaultValue=10000, parameterName=FarFade}, B=ScalarParameter() {defaultValue=1.1, parameterName=Buffer}) {constB=1.36, desc=add some buffer}
  n5 = Constant3Vector() {constant=<rgba 0 0 -999999 0>}
  n6 = Constant()
  n7 = Constant2Vector()
  n8 = ScalarParameter() {group=WPO, parameterName=BaseHight}
  n9 = VectorParameter() {group=WPO, parameterName=TextureWPO_UVtiling, defaultValue=<rgba 0.99 0.5 0 0>}
  n10 = Panner(Coordinate=Multiply(A=TextureCoordinate() {vTiling=0.5}, B=ComponentMask(None=@n9.RGB) {g=true, r=true})) {speedY=0.1}
  n11 = ComponentMask(None=TextureSampleParameter2D(UVs=@n10) {parameterName=WPO, samplerType=SAMPLERTYPE_Normal, texture=/Game/Effects/Fort_Effects/Textures/T_DustDistortion}.RGB) {g=true, r=true}
  n12 = Multiply(
   A=TextureSample(UVs=@n10) {texture=&T_Smooth_tile}.R
   B=ScalarParameter() {defaultValue=0.1, group=WPO, parameterName=WPO_Height}
  ) {constB=0.6}
  n13 = Transform(None=LinearInterpolate(
    A=AppendVector(A=@n7, B=@n8)
    B=AppendVector(A=@n11, B=@n12)
    Alpha=MaterialFunctionCall() {materialFunction=&LinearGradient}.VGradient
   ))
  n14 = ScalarParameter() {defaultValue=100, group=WPO, parameterName=WPO_OverallIntensity}
  If(
   A=@n1
   B=Constant() {r=1}
   "A > B"=@n2
   "A == B"=@n2
   "A < B"=Add(A=If(A=@n3.V2 Length, B=@n4, "A > B"=@n5, "A == B"=@n6, "A < B"=@n6), B=Multiply(A=@n13, B=@n14) {constB=80})
  )
  n15 = MaterialFunctionCall(Exponent=ScalarParameter() {defaultValue=5, parameterName="Exponent (S)"}) {materialFunction=&AxisAlignedFresnel}
  n16 = Panner(Coordinate=TextureCoordinate() {uTiling=6, vTiling=0.35}) {speedX=0.1, speedY=0.7}
  n17 = ScalarParameter() {defaultValue=0.5, parameterName="Rotation Angle (0-1) (S)"}
  n18 = Panner(Coordinate=TextureCoordinate() {uTiling=4, vTiling=0.2}) {speedX=-0.1, speedY=1}
  n19 = ScalarParameter() {defaultValue=0.5, parameterName="Rotation Angle (0-1) (S)"}
  n20 = LinearInterpolate(
   A=TextureSample(UVs=MaterialFunctionCall(UVs=@n16, "Rotation Angle (0-1)"=@n17) {materialFunction=&CustomRotator}.Rotated Values) {texture=&T_Smooth_tile}.R
   B=TextureSample(UVs=MaterialFunctionCall(UVs=@n18, "Rotation Angle (0-1)"=@n19) {materialFunction=&CustomRotator}.Rotated Values) {texture=&T_Inky_Smoke_Tile}.R
  )
  n21 = TextureSample(UVs=TextureCoordinate() {uTiling=2}) {samplerType=SAMPLERTYPE_Normal, texture=&T_DustDistortion}
  n22 = MaterialFunctionCall(Float2=TextureCoordinate()) {materialFunction=&BreakOutFloat2Components}
  n23 = ScalarParameter() {defaultValue=10, parameterName=U_Tile}
  n24 = Multiply(A=@n22.G, B=ScalarParameter() {defaultValue=1.2, parameterName=V_Tile})
  n25 = ScalarParameter() {defaultValue=1.5, parameterName=V_Tile_Exp}
  n26 = Add(
   A=Multiply(A=ComponentMask(None=@n21.RGB) {g=true, r=true}, B=ScalarParameter() {defaultValue=0.25, parameterName=DistMult})
   B=AppendVector(A=Multiply(A=@n22.R, B=@n23), B=Power(Base=@n24, Exp=@n25) {constExponent=3})
  )
  n27 = Time()
  n28 = AppendVector(A=ScalarParameter() {parameterName=SpeedX}, B=ScalarParameter() {defaultValue=1, parameterName=SpeedY})
  n29 = ScalarParameter() {defaultValue=0.5, parameterName="Rotation Angle (0-1) (S)"}
  n30 = MaterialFunctionCall(UVs=Panner(Coordinate=@n26, Time=@n27, Speed=@n28) {speedY=1}, "Rotation Angle (0-1)"=@n29) {materialFunction=&CustomRotator}
  n31 = TextureSample(UVs=@n30.Rotated Values) {texture=&T_NoisePack_01}
  n32 = ScalarParameter() {defaultValue=1.5, parameterName=Tiling_02}
  n33 = ScalarParameter() {defaultValue=0.5, parameterName=Speed_02_Mult}
  n34 = MaterialFunctionCall(UVs=Panner(Coordinate=Multiply(A=@n26, B=@n32), Time=@n27, Speed=Multiply(A=@n28, B=@n33)) {speedY=1}, "Rotation Angle (0-1)"=@n29) {materialFunction=&CustomRotator}
  n35 = Multiply(
   A=Add(A=@n31.G, B=@n31.G)
   B=LinearInterpolate(
    A=TextureSampleParameter2D(UVs=@n30.Rotated Values) {parameterName=NoiseB, texture=/Game/Global/Textures/Default/Caustics/T_Turbulence_Seamless}.G
    B=TextureSampleParameter2D(UVs=@n34.Rotated Values) {parameterName=NoiseA, texture=&T_ShieldCaustics}.G
   )
  )
  n36 = PixelDepth()
  n37 = Constant() {r=-100}
  n38 = Saturate(None=Divide(A=Add(A=@n36.R, B=@n37), B=ScalarParameter() {defaultValue=2500, parameterName=NearFade}))
  n39 = WorldPosition() {worldPositionShaderOffset=WPT_CameraRelative}
  n40 = Add(
   A=MaterialFunctionCall(Vector 2=ComponentMask(None=@n39.XYZ) {g=true, r=true}) {materialFunction=&VectorLength}.V2 Length
   B=Constant() {r=-100}
  )
  n41 = ScalarParameter() {defaultValue=10000, parameterName=FarFade}
  n42 = OneMinus(None=Saturate(None=Divide(A=@n40, B=@n41)))
  n43 = MaterialFunctionCall() {materialFunction=&LinearGradient}
  n44 = Multiply(A=Multiply(A=Multiply(A=@n20, B=@n35), B=Multiply(A=@n38, B=@n42)), B=Multiply(A=OneMinus(None=@n43.VGradient), B=@n43.VGradient))
  LinearInterpolate(A=Multiply(A=ScalarParameter() {defaultValue=5, parameterName=OpacityMult}, B=Multiply(A=@n15.Fresnel, B=@n44)), Alpha=@n1) {constB=0}
  LinearInterpolate(B=MaterialFunctionCall(
    Color_Morning=VectorParameter() {parameterName=MorningTint, defaultValue=<rgba 1.15 0.8884 0.805 0>}.RGB
    Color_Day=VectorParameter() {parameterName=DayTint, defaultValue=<rgba 0.2874 0.6512 1.2 0>}.RGB
    Color_Evening=VectorParameter() {parameterName=EveningTint, defaultValue=<rgba 1.1 0.7845 0.77 1>}.RGB
    Color_Night=VectorParameter() {parameterName=NightTint, defaultValue=<rgba 0.2656 0.85 0.85 0>}.RGB
   ) {materialFunction=&Color_Time_of_Day}.Result) {constA=1}
  Saturate(None=Add())
 ]
}
